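# Post-process style cosmetic AnimBP for the male medium DriedSilk body, driving its skeleton from the parent mesh it is attached to.
# Copy Pose From Mesh feeds the DriedSilk Control Rig, which receives the crouch state through bIsCrouching.
# An always-enabled additive bone-space Transform Bone then offsets and rotates the left pouch bone, dyn_pouch_l, before output.
# The event graph holds no logic.
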AnimBlueprint /BRCosmetics/Characters/Player/Male/Medium/Bodies/M_MED_DriedSilk/Meshes/M_MED_DriedSilk_AnimBP
target_skeleton=&/M_MED_DriedSilk_Skeleton
graphs [
 AnimGraph {nodes [
   n1 = Control Rig!(Source=Copy Pose From Mesh() {bUseAttachedParent=true}, Alpha, isCrouching:bool=Get bIsCrouching()) -> (Pose) {
    alphaBoolBlend {blendOption=Linear}
    controlRigAssetReference {blueprintRigClass=&/M_MED_DriedSilk_CtrlRig.M_MED_DriedSilk_CtrlRig_C}
   }
   n2 = "Transform (Modify) Bone - Bone: dyn_pouch_l"!(
    ComponentPose=Local To Component(LocalPose=@n1)
    bAlphaBoolEnabled:bool=True
    Translation="-1.332,-3.563,-0.945"
    Rotation="0.003,0.008,-18.073"
    Scale="1,1,1"
   ) -> (Pose) {
    alphaInputType=Bool
    rotationMode=BMM_Additive
    rotationSpace=BCS_BoneSpace
    translationMode=BMM_Additive
    translationSpace=BCS_BoneSpace
    alphaBoolBlend {blendInTime=0.3, blendOption=Sinusoidal, blendOutTime=0.3}
    boneToModify {boneName=dyn_pouch_l}
    rotation=<rot 0.003 0.008 -18.073>
    translation=<-1.332 -3.563 -0.945>
   }
   Output Pose(Result=Component To Local(ComponentPose=@n2)) {layerGroup=None}
  ]}
 EventGraph {nodes [Event BlueprintUpdateAnimation()]}
]
variables []
